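# Translucent unlit material for the blade tip effect on the Horned Judgment pickaxe, built for Niagara mesh particles.
# Particle color, optionally tinted by an inner-to-outer gradient from vertex green, feeds an emissive adjustment function.
# Panned blurred noise scales the vertex normal for world position offset, with intensity zero by default.
# Opacity multiplies particle alpha by a view-angle fade, then adds depth fade except on mobile.
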Material /BRCosmetics/Weapons/FORT_Melee/Pickaxe_HornedJudgment/FX/M_Pickaxe_HornedJudgement_BladeTip
graph {
 properties {
  bUsedWithNiagaraMeshParticles=true
  blendMode=BLEND_Translucent
  pixelDepthOffsetMode=PDOM_Legacy
  shadingModel=MSM_Unlit
  usedShadingModels=Unlit
 }
 nodes [
  n1 = ParticleColor()
  n2 = LinearInterpolate(
   A=VectorParameter() {group=Color, parameterName=Inner Color, defaultValue=<rgba 6 6 6 1>}.RGB
   B=VectorParameter() {group=Color, parameterName=Outer Color, defaultValue=<rgba 0.2 0.2 0.2 1>}.RGB
   Alpha=Power(Base=VertexColor().G, Exp=ScalarParameter() {defaultValue=0.2, group=Color, parameterName=Vertex Color Mask Exp})
  ) {constA=1, constB=0.25}
  MaterialFunctionCall(Emissive=LinearInterpolate(A=@n1.RGB, B=Multiply(A=@n1.RGB, B=@n2), Alpha=ScalarParameter() {group=Color, parameterName=Use Vertex Color for Emissive})) {materialFunction=&MF_FX_EmissiveAdjustments}
  n3 = Panner(
   Coordinate=Multiply(A=TextureCoordinate(), B=ScalarParameter() {defaultValue=1.5, group=WPO, parameterName=WPO Noise Scale})
   Speed=AppendVector(
    A=ScalarParameter() {defaultValue=1, group=WPO, parameterName=WPO Pan X}
    B=ScalarParameter() {defaultValue=2, group=WPO, parameterName=WPO Pan Y}
   )
  )
  Multiply(
   A=Multiply(
    A=VertexNormalWS()
    B=TextureSample(UVs=@n3) {samplerType=SAMPLERTYPE_Grayscale, texture=&LowResBlurredNoise}.R
   )
   B=ScalarParameter() {group=WPO, parameterName=WPO Intensity}
  )
  n4 = Multiply(
   A=@n1.A
   B=MaterialFunctionCall("FadeSensitivity (optional)"=ScalarParameter() {defaultValue=1, group=Opacity, parameterName=Fade Sensitivity, sliderMax=20}) {materialFunction=&FadeBasedOnViewAngle}.Result
  )
  ShadingPathSwitch(Default=DepthFade(Opacity=@n4, FadeDistance=ScalarParameter() {group=Opacity, parameterName=FadeDistance}), Mobile=@n4) {desc=Migrated from FeatureLevelSwitch}
 ]
}
