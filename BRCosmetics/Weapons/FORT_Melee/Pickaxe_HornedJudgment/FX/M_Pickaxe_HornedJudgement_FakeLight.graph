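# Additive unlit Niagara sprite material that fakes a light for Horned Judgment pickaxe effects, with depth testing disabled.
# Distance from the scene position behind the sprite to the particle center, scaled by radius and stretch, forms a sharpened radial falloff.
# Parameters blend in a UV gradient fade, tiled mask and panning dynamic mask.
# Particle color feeds time-of-day emissive adjustment, and Particle Stretch also scales a center-relative position offset.
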Material /BRCosmetics/Weapons/FORT_Melee/Pickaxe_HornedJudgment/FX/M_Pickaxe_HornedJudgement_FakeLight
graph {
 properties {
  bDisableDepthTest=true
  bUsedWithNiagaraSprites=true
  blendMode=BLEND_Additive
  pixelDepthOffsetMode=PDOM_Legacy
  shadingModel=MSM_Unlit
  translucencyPass=MTP_BeforeDOF
  usedShadingModels=Unlit
 }
 nodes [
  n1 = ParticlePositionWS()
  n2 = Distance(A=MaterialFunctionCall() {materialFunction=&WorldPositionBehindTranslucency}.Result, B=@n1)
  n3 = ScalarParameter() {defaultValue=1, parameterName=Particle Stretch}
  n4 = Multiply(A=ParticleRadius(), B=Add(A=@n3)) {constB=2}
  n5 = OneMinus(None=Saturate(None=Divide(A=@n2, B=@n4)))
  n6 = Multiply(A=@n5, B=@n5)
  n7 = Multiply(A=@n6, B=@n6)
  n8 = ComponentMask(None=TextureCoordinate()) {g=true}
  n9 = Power(
   Base=LinearInterpolate(A=@n8, B=OneMinus(None=@n8), Alpha=ScalarParameter() {parameterName=InvertFade}) {constAlpha=1}
   Exp=ScalarParameter() {defaultValue=2.2, parameterName=FadePower}
  ) {constExponent=2.2}
  n10 = LinearInterpolate(A=@n7, B=Multiply(A=@n7, B=Multiply(A=@n9)), Alpha=ScalarParameter() {parameterName=UseFade}) {constAlpha=1}
  n11 = TextureCoordinate()
  n12 = MaterialFunctionCall(R=ScalarParameter() {defaultValue=1, parameterName=UTiling Mask}, G=ScalarParameter() {defaultValue=1, parameterName=VTiling Mask}) {materialFunction=&AppendMany}
  n13 = Multiply(
   A=TextureSampleParameter2D(UVs=Multiply(A=@n11, B=@n12.RG)) {parameterName=Mask, samplerType=SAMPLERTYPE_Grayscale, texture=/Game/Effects/Fort_Effects/Textures/Tile/T_Concrete_Crack_A}.RGB
   B=@n10
  )
  n14 = ScalarParameter() {parameterName=U Panner Speed}
  n15 = ScalarParameter() {parameterName=V Panner Speed}
  n16 = TextureSampleParameter2D(UVs=Panner(
    Coordinate=TextureCoordinate()
    Time=Time()
    Speed=MaterialFunctionCall(R=@n14, G=@n15) {materialFunction=&AppendMany}.RG
   ) {speedY=1}) {parameterName=Dynamic Mask, texture=/Game/Effects/Fort_Effects/Textures/Taker/T_ThrobMask}
  LinearInterpolate(
   A=@n10
   B=LinearInterpolate(A=@n13, B=Multiply(A=@n16.RGB, B=@n13), Alpha=ScalarParameter() {parameterName=UseDynamicMask})
   Alpha=ScalarParameter() {parameterName=UseMask}
  )
  n17 = VectorParameter() {parameterName=TOD Emissive, defaultValue=<rgba 1 1 1 1>}
  MaterialFunctionCall(Emissive=Multiply(
    A=ParticleColor().RGB
    B=MaterialFunctionCall(Value_Morning=@n17.R, Value_Day=@n17.R, Value_Evening=@n17.R, Value_Night=@n17.R) {materialFunction=&Scalar_Time_of_Day}.Result
   )) {materialFunction=&MF_FX_EmissiveAdjustments}
  n18 = WorldPosition()
  Reroute(None=Multiply(A=Subtract(A=@n18.XYZ, B=@n1), B=@n3))
 ]
}
