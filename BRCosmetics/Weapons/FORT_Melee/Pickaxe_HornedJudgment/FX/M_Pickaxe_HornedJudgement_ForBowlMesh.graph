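# Translucent unlit Niagara mesh-particle material for wispy, toon-shaded flame on the Horned Judgment pickaxe's bowl mesh, blue by default.
# Two multiplied panning Voronoi samples distort scrolling flame noise, which is shaped by vertical gradients and particle alpha.
# That mask drives a curve-atlas three-color ramp, plus opacity with smoothstepped edges, depth fades and view-angle fade.
# World position offset combines a clamped particle-position bend, camera offset and normal-direction noise.
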Material /BRCosmetics/Weapons/FORT_Melee/Pickaxe_HornedJudgment/FX/M_Pickaxe_HornedJudgement_ForBowlMesh
graph {
 properties {
  bUsedWithNiagaraMeshParticles=true
  blendMode=BLEND_Translucent
  pixelDepthOffsetMode=PDOM_Legacy
  shadingModel=MSM_Unlit
  usedShadingModels=Unlit
 }
 nodes [
  n1 = ScalarParameter() {group=Color, parameterName=Color Smooth Step Min, sliderMax=1, sliderMin=-1}
  n2 = ScalarParameter() {defaultValue=2, group=Color, parameterName=Color Smooth Step Max, sliderMax=5}
  n3 = NamedRerouteUsage()
  n4 = ScalarParameter() {group=Wispy Flame, parameterName=Wispy Distortion Strength, sliderMax=1, sliderMin=-1}
  n5 = Multiply(
   A=ScalarParameter() {defaultValue=1, group=Wispy Flame, parameterName=Scale Y, sliderMax=5}
   B=ComponentMask(None=TextureCoordinate()) {g=true}
  )
  n6 = ScalarParameter() {defaultValue=0.5, group=Wispy Flame, parameterName=Scale Y Add, sliderMax=1}
  n7 = Multiply(
   A=Add(A=TextureCoordinate(), B=Multiply(A=@n3, B=@n4) {constB=0.1})
   B=AppendVector(A=ScalarParameter() {defaultValue=1, group=Wispy Flame, parameterName=Scale X, sliderMax=5}, B=Add(A=@n5, B=@n6) {constB=0.5})
  )
  n8 = AppendVector(
   A=ScalarParameter() {group=Wispy Flame, parameterName=Speed X, sliderMax=5, sliderMin=-5}
   B=ScalarParameter() {defaultValue=1, group=Wispy Flame, parameterName=Speed Y, sliderMax=5, sliderMin=-5}
  )
  n9 = DynamicParameter() {defaultValue=<rgba 0 1 0 1>, paramNames [UV1 U Offset, Dive, Camera Offset, Param4]}
  n10 = Constant()
  n11 = TextureSampleParameter2D(UVs=Add(A=Panner(Coordinate=@n7, Speed=@n8) {speedX=1}, B=AppendVector(A=@n9.UV1 U Offset, B=@n10))) {group=Wispy Flame, parameterName=Flame Texture, texture=/Game/Effects/Fort_Effects/Textures/T_WispyFireNoise03}
  n12 = Power(
   Base=ComponentMask(None=TextureCoordinate()) {g=true}
   Exp=ScalarParameter() {defaultValue=1, group=Wispy Flame, parameterName=Top fade Power, sliderMax=10}
  ) {constExponent=3}
  n13 = Multiply(A=ScalarParameter() {defaultValue=1, group=Wispy Flame, parameterName=Top Fade Multiply, sliderMax=5}, B=@n12)
  n14 = ScalarParameter() {group=Shape Lerp, parameterName=Shape Lerp Min, sliderMax=2, sliderMin=-2}
  n15 = ScalarParameter() {defaultValue=1, group=Shape Lerp, parameterName=Shape Lerp Max, sliderMax=2, sliderMin=-2}
  n16 = Power(
   Base=ComponentMask(None=TextureCoordinate() {uTiling=4}) {g=true}
   Exp=ScalarParameter() {defaultValue=2, group=Shape Lerp, parameterName=Shape Lerp Power, sliderMax=10}
  ) {constExponent=1}
  n17 = Multiply(A=ParticleColor().A, B=LinearInterpolate(A=Multiply(A=@n11.R, B=@n13), B=@n12, Alpha=SmoothStep(Min=@n14, Max=@n15, Value=@n16)))
  n18 = Power(Base=Saturate(None=@n17), Exp=ScalarParameter() {defaultValue=1, group=Color, parameterName=Color Power, sliderMax=5}) {constExponent=1}
  n19 = CurveAtlasRowParameter(CurveTime=LinearInterpolate(B=SmoothStep(Min=@n1, Max=@n2, Value=@n18) {constMax=2}) {constA=1, constAlpha=1}) {atlas=/Game/Athena/LinearColorCurves/VFX_CurveAtlas_Toon, curve=/Game/Athena/LinearColorCurves/Curve_FireToon02, parameterName=Curve_Color}
  MaterialFunctionCall(Emissive=MaterialFunctionCall(
    1=VectorParameter() {parameterName=Color 1, defaultValue=<rgba 0 0.055 1 1>}.RGB
    2=VectorParameter() {parameterName=Color 2, defaultValue=<rgba 0.2213 0.2737 0.625 1>}.RGB
    3=VectorParameter() {parameterName=Color 3, defaultValue=<rgba 0.4327 1.6448 5 1>}.RGB
    A=@n19.R
   ) {materialFunction=&Lerp_Multiple_Float3}.Lerp 3 Inputs) {materialFunction=&MF_FX_EmissiveAdjustments}
  n20 = VectorParameter() {parameterName=ViewBasedFadeLocalAxis, defaultValue=<rgba 0 0 1 1>}
  n21 = CameraVectorWS()
  n22 = Abs(None=DotProduct(A=Transform(None=@n20.RGB) {transformSourceType=TRANSFORMSOURCE_Local}, B=@n21))
  n23 = Abs(None=DotProduct(A=VertexNormalWS(), B=@n21))
  n24 = SmoothStep(
   Min=ScalarParameter() {parameterName=ViewBasedFadeMin, sliderMax=2, sliderMin=-2}
   Max=ScalarParameter() {defaultValue=1, parameterName=ViewBasedFadeMax, sliderMax=2, sliderMin=-2}
   Value=Power(Base=Min(A=@n22, B=@n23), Exp=ScalarParameter() {defaultValue=8, parameterName=ViewBasedFadeSensitivity, sliderMax=16})
  )
  n25 = Multiply(A=@n17)
  n26 = DepthFade(Opacity=@n25, FadeDistance=ScalarParameter() {defaultValue=35, parameterName=Depth Fade Distance Before Hard Edge, sliderMax=100}) {fadeDistanceDefault=25}
  n27 = SmoothStep(
   Min=ScalarParameter() {parameterName=Edge Hardness Min, sliderMax=1}
   Max=ScalarParameter() {defaultValue=0.9, parameterName=Edge Hardness Max, sliderMax=1}
   Value=ShadingPathSwitch(Default=@n26, Deferred=@n26, Forward=@n25, Mobile=@n25)
  ) {constMax=0.1}
  n28 = DepthFade(Opacity=@n27, FadeDistance=ScalarParameter() {defaultValue=35, parameterName=Depth Fade Distance, sliderMax=100}) {fadeDistanceDefault=25}
  Multiply(
   A=ScalarParameter() {defaultValue=1, parameterName=Overall Opacity, sliderMax=1}
   B=Multiply(A=Clamp(None=@n24), B=ShadingPathSwitch(Default=@n28, Deferred=@n28, Forward=@n27, Mobile=@n27)) {constA=1}
  ) {constA=1, constB=-1}
  n29 = TextureCoordinate()
  n30 = AppendVector(
   A=ScalarParameter() {group=UV Distortion, parameterName=UV Speed X 2, sliderMax=5, sliderMin=-5}
   B=ScalarParameter() {defaultValue=1, group=UV Distortion, parameterName=UV Speed Y 2, sliderMax=5, sliderMin=-5}
  )
  n31 = ScalarParameter() {defaultValue=1, group=UV Distortion, parameterName=UV Scale X 2, sliderMax=5}
  n32 = ScalarParameter() {defaultValue=0.25, group=UV Distortion, parameterName=UV Scale Y 2, sliderMax=5}
  n33 = TextureSampleParameter2D(UVs=Multiply(A=Panner(Coordinate=@n29, Speed=@n30) {speedX=-0.5}, B=AppendVector(A=@n31, B=@n32))) {group=UV Distortion, parameterName=Distortion Texture, texture=/Game/Effects/Fort_Effects/Textures/T_PinkSpike_Voronoi2D_Height}
  n34 = TextureCoordinate()
  n35 = AppendVector(
   A=ScalarParameter() {group=UV Distortion, parameterName=UV Speed X, sliderMax=5, sliderMin=-5}
   B=ScalarParameter() {defaultValue=2, group=UV Distortion, parameterName=UV Speed Y, sliderMax=5, sliderMin=-5}
  )
  n36 = ScalarParameter() {defaultValue=1, group=UV Distortion, parameterName=UV Scale X, sliderMax=5}
  n37 = ScalarParameter() {defaultValue=0.5, group=UV Distortion, parameterName=UV Scale Y, sliderMax=5}
  n38 = TextureSampleParameter2D(UVs=Multiply(A=Panner(Coordinate=@n34, Speed=@n35) {speedX=-0.5}, B=AppendVector(A=@n36, B=@n37))) {group=UV Distortion, parameterName=Distortion Texture, texture=&T_PinkSpike_Voronoi2D_Height}
  n39 = Multiply(A=@n33.RGB, B=@n38.RGB)
  NamedRerouteDeclaration(None=Subtract(A=ComponentMask(None=@n39) {g=true, r=true}) {constB=0.25}) {name=Distortion, nodeColor=<rgba 1 0.3281 0 1>}
  n40 = VectorParameter() {parameterName=WPO Axis Mask, defaultValue=<rgba 1 1 1 1>}
  n41 = DynamicParameter() {parameterIndex=1, paramNames [Position X, Position Y, Position Z, Param4]}
  n42 = ScalarParameter() {defaultValue=20, parameterName=WPO Max Offset, sliderMax=50}
  n43 = Clamp(
   None=Subtract(
    A=MaterialFunctionCall(X=@n41.Position X, Y=@n41.Position X, Z=@n41.Position X) {materialFunction=&MakeFloat3}.Result
    B=ParticlePositionWS()
   )
   Min=Multiply(A=@n42) {constB=-1}
   Max=@n42
  )
  n44 = OneMinus(None=ComponentMask(None=TextureCoordinate() {coordinateIndex=1}) {g=true})
  n45 = ScalarParameter() {defaultValue=1, parameterName=WPO Bend, sliderMax=10}
  n46 = Multiply(
   A=ScalarParameter() {parameterName=WPO Intensity, sliderMax=100}
   B=LinearInterpolate(B=Multiply(A=@n40.RGB, B=@n43), Alpha=Power(Base=@n44, Exp=@n45))
  )
  n47 = MaterialFunctionCall(Offset Amount=ScalarParameter() {parameterName=Camera Offset}) {materialFunction=/Engine/Functions/Engine_MaterialFunctions02/WorldPositionOffset/CameraOffset}
  n48 = Multiply(A=ComponentMask(None=NamedRerouteUsage()) {r=true}, B=VertexNormalWS())
  n49 = ScalarParameter() {parameterName=WPO Noise Intensity, sliderMax=100}
  NamedRerouteDeclaration(None=Add(A=Add(A=@n46, B=@n47.Camera Offset), B=Multiply(A=@n48, B=@n49))) {name=WPO, nodeColor=<rgba 0 0.3437 1 1>}
  NamedRerouteUsage()
 ]
}
