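# Unlit translucent Niagara sprite and mesh particle material for the idle spike effect on the Horned Judgement pickaxe.
# UVs taper toward both ends of one axis, then pan and tile a spike texture.
# Its curve-atlas-remapped red channel blends three colors, unless the Glow switch substitutes particle color.
# Particle and texture alpha fade by view angle, and time-driven sine waves, masked by UVs, displace vertices along a transformed vector.
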Material /BRCosmetics/Weapons/FORT_Melee/Pickaxe_HornedJudgment/FX/M_Pickaxe_HornedJudgement_IdleSpikes
graph {
 properties {
  bUsedWithNiagaraMeshParticles=true
  bUsedWithNiagaraSprites=true
  blendMode=BLEND_Translucent
  pixelDepthOffsetMode=PDOM_Legacy
  shadingModel=MSM_Unlit
  usedShadingModels=Unlit
 }
 nodes [
  n1 = ParticleColor()
  n2 = VectorParameter() {parameterName=Color 1, defaultValue=<rgba 0 0.055 1 1>}
  n3 = VectorParameter() {parameterName=Color 2, defaultValue=<rgba 0.2213 0.2737 0.625 1>}
  n4 = VectorParameter() {parameterName=Color 3, defaultValue=<rgba 0.4327 1.6448 5 1>}
  n5 = MaterialFunctionCall(XY=TextureCoordinate()) {materialFunction=&Swizzle}
  n6 = ComponentMask(None=TextureCoordinate()) {g=true}
  n7 = ScalarParameter() {defaultValue=1, parameterName=Start Shirnk, sliderMax=1}
  n8 = OneMinus(None=@n6)
  n9 = ScalarParameter() {defaultValue=1, parameterName=End Shirnk, sliderMax=1}
  n10 = AppendVector(A=Constant() {r=1}, B=Multiply(A=Power(Base=@n6, Exp=@n7) {constExponent=0.3}, B=Power(Base=@n8, Exp=@n9) {constExponent=0.3}) {constA=1})
  n11 = Saturate(None=Add(A=Divide(A=@n5.YX, B=@n10), B=Constant2Vector()))
  n12 = AppendVector(A=ScalarParameter() {defaultValue=0.5, parameterName=Speed X, sliderMax=3}, B=ScalarParameter() {parameterName=Speed Y, sliderMax=3})
  n13 = ScalarParameter() {defaultValue=1, parameterName=Tile X, sliderMax=10}
  n14 = ScalarParameter() {defaultValue=1, parameterName=Tile Y, sliderMax=10}
  n15 = TextureSampleParameter2D(UVs=Multiply(A=Panner(Coordinate=@n11, Speed=@n12) {speedX=0.5}, B=AppendVector(A=@n13, B=@n14))) {parameterName=Texture, texture=&/T_Pickaxe_HornedJudgement_Spikes}
  n16 = CurveAtlasRowParameter(CurveTime=Power(Base=@n15.R, Exp=ScalarParameter() {defaultValue=1, parameterName=Color Power, sliderMax=5})) {
   atlas=/Game/Weapons/FORT_Melee/Pickaxe_HornedJudgment/FX/CurveAtlas_HornedJudgement
   curve=/Game/Weapons/FORT_Melee/Pickaxe_HornedJudgment/FX/Curve_Pickaxe_HornedJudgement_Spike
   parameterName=Curve_Color
  }
  MaterialFunctionCall(Emissive=StaticSwitchParameter(
    True=@n1.RGB
    False=MaterialFunctionCall(1=@n2.RGB, 2=@n3.RGB, 3=@n4.RGB, A=@n16.R) {materialFunction=&Lerp_Multiple_Float3}.Lerp 3 Inputs
   ) {parameterName=Glow?}) {materialFunction=&MF_FX_EmissiveAdjustments}
  Multiply(
   A=Multiply(A=@n1.A, B=@n15.A)
   B=MaterialFunctionCall("FadeSensitivity (optional)"=ScalarParameter() {defaultValue=4, parameterName=Fade Sensitivity, sliderMax=20}) {materialFunction=&FadeBasedOnViewAngle}.Result
  )
  n17 = TextureCoordinate()
  n18 = ComponentMask(None=@n17) {r=true}
  n19 = ComponentMask(None=@n17) {g=true}
  n20 = Sine(None=@n19) {period=2}
  n21 = Add(A=@n19, B=Time())
  n22 = ScalarParameter() {defaultValue=1, parameterName=Period, sliderMax=10}
  n23 = ConstantBiasScale(None=Sine(None=Multiply(A=@n21, B=@n22)))
  n24 = ScalarParameter() {defaultValue=5, parameterName=WPO Intensity, sliderMax=50}
  Multiply(
   A=Multiply(A=Multiply(A=@n18, B=@n20), B=Multiply(A=@n23, B=@n24))
   B=Transform(None=VectorParameter() {parameterName=WPO Tangent Vector, defaultValue=<rgba 0 0 1 1>}.RGB)
  )
 ]
}
