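# Translucent unlit, two-sided Niagara material for the Horned Judgment pickaxe's impact flame.
# Panning polar-UV noise distorts both a vertically biased radial mask and a second panning polar flame texture, which erodes that mask.
# The result drives a curve-atlas remap into a three-color emissive gradient, blue by default.
# Opacity adds particle alpha and deferred-path depth fade, then a smoothstep edge, with dynamic parameters controlling radius, bias, distortion and edge hardness.
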Material /BRCosmetics/Weapons/FORT_Melee/Pickaxe_HornedJudgment/FX/M_Pickaxe_HornedJudgement_ImpactFlame
graph {
 properties {
  bUsedWithNiagaraMeshParticles=true
  bUsedWithNiagaraRibbons=true
  bUsedWithNiagaraSprites=true
  blendMode=BLEND_Translucent
  pixelDepthOffsetMode=PDOM_Legacy
  shadingModel=MSM_Unlit
  twoSided=true
  usedShadingModels=Unlit
 }
 nodes [
  n1 = TextureCoordinate()
  n2 = AppendVector(
   A=ScalarParameter() {defaultValue=2, parameterName=Distortion Scale X, sliderMax=5}
   B=ScalarParameter() {defaultValue=0.5, parameterName=Distortion Scale Y, sliderMax=5}
  )
  n3 = MaterialFunctionCall!(UV, Twist Center, Twist Edges) {materialFunction=/Game/Accessories/FORT_Backpacks/Backpack_F_MED_Celestial_Glow/FX/Materials/MF_PolarUV}
  n4 = ScalarParameter() {parameterName=Distortion Speed X, sliderMax=5}
  n5 = ScalarParameter() {defaultValue=-0.5, parameterName=Distortion Speed Y, sliderMax=5}
  n6 = TextureSampleParameter2D!(
   UVs=Panner(Coordinate=Multiply(A=@n2, B=@n3.Polar UVs) {constB=2}, Speed=AppendVector(A=@n4, B=@n5)) {speedY=-1}
   "DDX(UVs)"=@n3.DDX
   "DDY(UVs)"=@n3.DDX
   Apply View MipBias
  ) {
   mipValueMode=TMVM_Derivative
   parameterName=Distortion Texture
   samplerType=SAMPLERTYPE_Grayscale
   texture=/Game/Effects/Fort_Effects/Textures/Fire/T_WispyFireNoise_Soft_M
  }
  n7 = Multiply(
   A=Subtract(A=@n6.R) {constB=0.5}
   B=DynamicParameter() {defaultValue=<rgba 0.15 0.1 0.5 0.1>, paramNames [Radius, Thickness, Hard Edge, Distortion]}.Distortion
  ) {constB=0}
  n8 = Constant() {r=1}
  n9 = Power(
   Base=ComponentMask(None=TextureCoordinate()) {g=true}
   Exp=DynamicParameter() {parameterIndex=1, defaultValue=<rgba 0 1 1 1>, paramNames [Bias,Param2,Param3,Param4]}.Bias
  ) {constExponent=0.25}
  n10 = Distance(A=Multiply(A=Add(A=@n1, B=@n7), B=AppendVector(A=@n8, B=@n9)), B=Constant() {r=0.5})
  n11 = DynamicParameter() {defaultValue=<rgba 0.15 0.1 0.5 0.1>, paramNames [Radius, Thickness, Hard Edge, Distortion]}
  n12 = ScalarParameter() {defaultValue=0.25, parameterName=Hardness}
  n13 = MaterialFunctionCall() {materialFunction=&MF_PolarUV}
  n14 = Add(A=@n13.Polar UVs, B=@n7)
  n15 = AppendVector(
   A=ScalarParameter() {defaultValue=1, parameterName=Scale X, sliderMax=5}
   B=ScalarParameter() {defaultValue=1, parameterName=Scale Y, sliderMax=5}
  )
  n16 = ScalarParameter() {parameterName=Speed X, sliderMax=5}
  n17 = ScalarParameter() {defaultValue=-0.5, parameterName=Speed Y, sliderMax=5}
  n18 = TextureSampleParameter2D(
   UVs=Panner(Coordinate=Multiply(A=@n14, B=@n15) {constB=2}, Speed=AppendVector(A=@n16, B=@n17)) {speedY=-0.5}
   "DDX(UVs)"=@n13.DDX
   "DDY(UVs)"=@n13.DDY
  ) {
   mipValueMode=TMVM_Derivative
   parameterName=Flame Texture
   samplerType=SAMPLERTYPE_Grayscale
   texture=&T_WispyFireNoise_Soft_M
  }
  n19 = Multiply(
   A=Subtract(A=SphereMask(A=@n10, B=@n11.Radius, Radius=@n11.Radius, Hardness=@n12) {attenuationRadius=0.1, hardnessPercent=0}, B=@n18.R)
   B=@n18.R
  )
  n20 = ScalarParameter() {defaultValue=1, parameterName=Color Power, sliderMax=5}
  n21 = CurveAtlasRowParameter(CurveTime=SmoothStep(
    Min=ScalarParameter() {defaultValue=0.3, parameterName=Color SmoothStep Min, sliderMax=3, sliderMin=-3}
    Max=ScalarParameter() {defaultValue=1.2, parameterName=Color SmoothStep Max, sliderMax=3, sliderMin=-3}
    Value=Power(Base=@n19, Exp=@n20)
   ) {constMax=1.2, constMin=0.3}) {
   atlas=/Game/Weapons/FORT_Melee/Pickaxe_HornedJudgment/FX/CurveAtlas_HornedJudgement
   curve=/Game/Weapons/FORT_Melee/Pickaxe_HornedJudgment/FX/Curve_Pickaxe_HornedJudgement_BladeTip
   defaultValue=1
  }
  MaterialFunctionCall(Emissive=MaterialFunctionCall(
    1=VectorParameter() {parameterName=Color 1, defaultValue=<rgba 0 0.055 1 1>}.RGB
    2=VectorParameter() {parameterName=Color 2, defaultValue=<rgba 0.2213 0.2737 0.625 1>}.RGB
    3=VectorParameter() {parameterName=Color 3, defaultValue=<rgba 0.4327 1.6448 5 1>}.RGB
    A=@n21.R
   ) {materialFunction=&Lerp_Multiple_Float3}.Lerp 3 Inputs) {materialFunction=&MF_FX_EmissiveAdjustments}
  n22 = Multiply(A=ParticleColor().A, B=@n19)
  n23 = DepthFade(Opacity=@n22, FadeDistance=ScalarParameter() {defaultValue=35, parameterName=Depth Fade Distance, sliderMax=100})
  SmoothStep(
   Max=DynamicParameter() {defaultValue=<rgba 0.15 0.1 0.5 0.1>, paramNames [Radius, Thickness, Hard Edge, Distortion]}.Hard Edge
   Value=ShadingPathSwitch(Default=@n23, Deferred=@n23, Forward=@n22, Mobile=@n22)
  )
 ]
}
