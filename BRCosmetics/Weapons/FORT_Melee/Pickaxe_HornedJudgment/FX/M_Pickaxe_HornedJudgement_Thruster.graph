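# Translucent unlit Niagara material for the thruster flame effect on the Horned Judgment pickaxe.
# Two panned noise textures distort a scrolling wispy flame texture, which blends into selectable candle, vertex-color or sphere masks.
# That shape drives a curve-atlas three-color gradient and opacity with depth, view-angle and dynamic-parameter fades.
# World position offset combines a clamped bend toward a dynamic-parameter value, camera offset, center puff and camera-aligned sine wiggle.
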Material /BRCosmetics/Weapons/FORT_Melee/Pickaxe_HornedJudgment/FX/M_Pickaxe_HornedJudgement_Thruster
graph {
 properties {
  bUsedWithNiagaraMeshParticles=true
  bUsedWithNiagaraRibbons=true
  bUsedWithNiagaraSprites=true
  blendMode=BLEND_Translucent
  pixelDepthOffsetMode=PDOM_Legacy
  shadingModel=MSM_Unlit
  twoSided=true
  usedShadingModels=Unlit
 }
 nodes [
  n1 = SmoothStep(Value=ComponentMask(None=TextureCoordinate()) {g=true}) {constMax=0.9}
  n2 = ScalarParameter() {parameterName=End Value Fade}
  n3 = ScalarParameter() {group=Color, parameterName=Color Smooth Step Min, sliderMax=1, sliderMin=-1}
  n4 = ScalarParameter() {defaultValue=2, group=Color, parameterName=Color Smooth Step Max, sliderMax=5}
  n5 = NamedRerouteUsage()
  n6 = ScalarParameter() {group=Wispy Flame, parameterName=Wispy Distortion Strength, sliderMax=1, sliderMin=-1}
  n7 = TextureCoordinate() {coordinateIndex=1}
  n8 = Multiply(
   A=ScalarParameter() {defaultValue=1, group=Wispy Flame, parameterName=Scale Y, sliderMax=5}
   B=Multiply(A=ComponentMask(None=@n7) {g=true}, B=ScalarParameter() {defaultValue=1, parameterName=Flame Top UV Stretch})
  )
  n9 = Multiply(
   A=Add(A=TextureCoordinate() {coordinateIndex=1}, B=Multiply(A=@n5, B=@n6) {constB=0.1})
   B=AppendVector(A=ScalarParameter() {defaultValue=1, group=Wispy Flame, parameterName=Scale X, sliderMax=5}, B=Add(A=@n8) {constB=0.5})
  )
  n10 = Time()
  n11 = AppendVector(
   A=ScalarParameter() {group=Wispy Flame, parameterName=Speed X, sliderMax=5, sliderMin=-5}
   B=ScalarParameter() {defaultValue=1, group=Wispy Flame, parameterName=Speed Y, sliderMax=5, sliderMin=-5}
  )
  n12 = DynamicParameter() {defaultValue=<rgba 0 1 0 1>, paramNames [UV1 U Offset, Dive, Camera Offset, Param4]}
  n13 = Constant()
  n14 = TextureSampleParameter2D(UVs=Add(A=Panner(Coordinate=@n9, Time=@n10, Speed=@n11) {speedX=1}, B=AppendVector(A=@n12.UV1 U Offset, B=@n13))) {
   group=Wispy Flame
   parameterName=Flame Texture
   samplerType=SAMPLERTYPE_Grayscale
   texture=/Game/Effects/Fort_Effects/Textures/Fire/T_WispyFireNoise_Soft_M
  }
  n15 = Sine(None=ComponentMask(None=TextureCoordinate() {coordinateIndex=1}) {g=true}) {period=2}
  n16 = TextureCoordinate()
  n17 = Power(Base=OneMinus(None=ComponentMask(None=@n16) {r=true}), Exp=ScalarParameter() {defaultValue=2, parameterName=Candle Mask Power})
  n18 = ScalarParameter() {defaultValue=0.5, parameterName=Sphere Mask Position X}
  n19 = ScalarParameter() {defaultValue=0.5, parameterName=Sphere Mask Position Y}
  n20 = LinearInterpolate(
   A=LinearInterpolate(A=Multiply(A=@n15, B=@n17), B=VertexColor().R, Alpha=ScalarParameter() {parameterName=Use Vert Color As Mask})
   B=SphereMask(A=TextureCoordinate(), B=AppendVector(A=@n18, B=@n19)) {attenuationRadius=0.5, hardnessPercent=0}
   Alpha=ScalarParameter() {parameterName=Use Sphere Mask}
  )
  n21 = Multiply(A=ScalarParameter() {defaultValue=1, group=Wispy Flame, parameterName=Candle Shape Mask, sliderMax=5}, B=@n20)
  n22 = ScalarParameter() {group=Shape Lerp, parameterName=Shape Lerp Min, sliderMax=2, sliderMin=-2}
  n23 = ScalarParameter() {defaultValue=1, group=Shape Lerp, parameterName=Shape Lerp Max, sliderMax=2, sliderMin=-2}
  n24 = Power(
   Base=ComponentMask(None=TextureCoordinate() {uTiling=4}) {g=true}
   Exp=ScalarParameter() {defaultValue=2, group=Shape Lerp, parameterName=Shape Lerp Power, sliderMax=10}
  ) {constExponent=1}
  n25 = Multiply(A=ParticleColor().A, B=LinearInterpolate(A=Multiply(A=@n14.R, B=@n21), B=@n20, Alpha=SmoothStep(Min=@n22, Max=@n23, Value=@n24)))
  n26 = Power(Base=Saturate(None=@n25), Exp=ScalarParameter() {defaultValue=1, group=Color, parameterName=Color Power, sliderMax=5}) {constExponent=1}
  n27 = CurveAtlasRowParameter(CurveTime=Multiply(A=LinearInterpolate(B=@n1, Alpha=@n2) {constA=1}, B=SmoothStep(Min=@n3, Max=@n4, Value=@n26) {constMax=2})) {
   atlas=/Game/Athena/LinearColorCurves/VFX_CurveAtlas_Toon
   curve=/Game/Weapons/FORT_Melee/Pickaxe_HornedJudgment/FX/Curve_HornedJudgement_Thruster_Gradient
   defaultValue=8
   parameterName=Curve_Color
  }
  MaterialFunctionCall(Emissive=MaterialFunctionCall(
    1=VectorParameter() {parameterName=Color 1, defaultValue=<rgba 0 0.055 1 1>}.RGB
    2=VectorParameter() {parameterName=Color 2, defaultValue=<rgba 0.2213 0.2737 0.625 1>}.RGB
    3=VectorParameter() {parameterName=Color 3, defaultValue=<rgba 0.4327 1.6448 5 1>}.RGB
    A=@n27.R
   ) {materialFunction=&Lerp_Multiple_Float3}.Lerp 3 Inputs) {materialFunction=&MF_FX_EmissiveAdjustments}
  n28 = ScalarParameter() {defaultValue=8, parameterName=ViewBasedFadeSensitivity, sliderMax=16}
  n29 = Multiply(A=@n25)
  n30 = DepthFade(Opacity=@n29, FadeDistance=ScalarParameter() {defaultValue=35, parameterName=Depth Fade Distance Before Hard Edge, sliderMax=100}) {fadeDistanceDefault=25}
  n31 = SmoothStep(
   Min=ScalarParameter() {parameterName=Edge Hardness Min, sliderMax=1}
   Max=ScalarParameter() {defaultValue=0.9, parameterName=Edge Hardness Max, sliderMax=1}
   Value=ShadingPathSwitch(Default=@n30, Deferred=@n30, Forward=@n29, Mobile=@n29)
  ) {constMax=0.1}
  n32 = DepthFade(Opacity=@n31, FadeDistance=ScalarParameter() {defaultValue=35, parameterName=Depth Fade Distance, sliderMax=100}) {fadeDistanceDefault=25}
  Multiply(
   A=Multiply(
    A=MaterialFunctionCall("FadeSensitivity (optional)"=@n28) {materialFunction=&FadeBasedOnViewAngle}.Result
    B=ShadingPathSwitch(Default=@n32, Deferred=@n32, Forward=@n31, Mobile=@n31)
   )
   B=DynamicParameter() {parameterIndex=2, paramNames [Overall Opacity, Param2, Param3, Param4]}.Overall Opacity
  )
  n33 = TextureCoordinate() {coordinateIndex=1}
  n34 = Time()
  n35 = AppendVector(
   A=ScalarParameter() {group=UV Distortion, parameterName=UV Speed X 2, sliderMax=5, sliderMin=-5}
   B=ScalarParameter() {defaultValue=1, group=UV Distortion, parameterName=UV Speed Y 2, sliderMax=5, sliderMin=-5}
  )
  n36 = ScalarParameter() {defaultValue=1, group=UV Distortion, parameterName=UV Scale X 2, sliderMax=5}
  n37 = ScalarParameter() {defaultValue=0.25, group=UV Distortion, parameterName=UV Scale Y 2, sliderMax=5}
  n38 = TextureSampleParameter2D(UVs=Multiply(A=Panner(Coordinate=@n33, Time=@n34, Speed=@n35) {speedX=-0.5}, B=AppendVector(A=@n36, B=@n37))) {group=UV Distortion, parameterName=Distortion Texture, samplerType=SAMPLERTYPE_Grayscale, texture=/FortWater/Textures/LowResBlurredNoise}
  n39 = TextureCoordinate() {coordinateIndex=1}
  n40 = Time()
  n41 = AppendVector(
   A=ScalarParameter() {group=UV Distortion, parameterName=UV Speed X, sliderMax=5, sliderMin=-5}
   B=ScalarParameter() {defaultValue=2, group=UV Distortion, parameterName=UV Speed Y, sliderMax=5, sliderMin=-5}
  )
  n42 = ScalarParameter() {defaultValue=1, group=UV Distortion, parameterName=UV Scale X, sliderMax=5}
  n43 = ScalarParameter() {defaultValue=0.5, group=UV Distortion, parameterName=UV Scale Y, sliderMax=5}
  n44 = TextureSampleParameter2D(UVs=Multiply(A=Panner(Coordinate=@n39, Time=@n40, Speed=@n41) {speedX=-0.5}, B=AppendVector(A=@n42, B=@n43))) {group=UV Distortion, parameterName=Distortion Texture, samplerType=SAMPLERTYPE_Grayscale, texture=&LowResBlurredNoise}
  n45 = Multiply(A=@n38.RGB, B=@n44.RGB)
  NamedRerouteDeclaration(None=Subtract(A=ComponentMask(None=@n45) {g=true, r=true}) {constB=0.25}) {name=Distortion, nodeColor=<rgba 1 0.3281 0 1>}
  n46 = DynamicParameter() {parameterIndex=1, paramNames [X,Y,Z,Param4]}
  n47 = MaterialFunctionCall(X=@n46.X, Y=@n46.X, Z=@n46.X) {materialFunction=&MakeFloat3}
  n48 = ParticlePositionWS()
  n49 = ScalarParameter() {defaultValue=20, parameterName=WPO Max Offset, sliderMax=50}
  n50 = Multiply(
   A=VectorParameter() {parameterName=WPO Axis Mask, defaultValue=<rgba 1 1 1 1>}.RGB
   B=Clamp(None=Subtract(A=@n47.Result, B=@n48), Min=Multiply(A=@n49) {constB=-1}, Max=@n49)
  )
  n51 = TextureCoordinate() {coordinateIndex=1}
  n52 = Power(Base=OneMinus(None=ComponentMask(None=@n51) {g=true}), Exp=ScalarParameter() {defaultValue=1, parameterName=WPO Bend, sliderMax=10})
  n53 = Add(
   A=Multiply(A=ScalarParameter() {parameterName=WPO Intensity, sliderMax=100}, B=LinearInterpolate(B=@n50, Alpha=@n52))
   B=MaterialFunctionCall(Offset Amount=ScalarParameter() {parameterName=Camera Offset}) {materialFunction=/Engine/Functions/Engine_MaterialFunctions02/WorldPositionOffset/CameraOffset}.Camera Offset
  )
  n54 = OneMinus(None=ComponentMask(None=TextureCoordinate()) {r=true})
  n55 = Constant3Vector() {constant=<rgba 0 0 1 0>}
  n56 = Multiply(
   A=Multiply(A=SmoothStep(Value=@n54) {constMax=0.9}, B=Transform(None=@n55) {transformSourceType=TRANSFORMSOURCE_Instance})
   B=ScalarParameter() {parameterName=Center Puff}
  )
  n57 = ComponentMask(None=TextureCoordinate() {coordinateIndex=1}) {g=true}
  n58 = Time()
  n59 = ScalarParameter() {defaultValue=1, group=WPO Wiggle, parameterName=Wiggle Speed, sliderMax=5, sliderMin=-5}
  n60 = Multiply(
   A=Add(A=@n57, B=Multiply(A=@n58, B=@n59))
   B=ScalarParameter() {defaultValue=2, group=WPO Wiggle, parameterName=Wiggle Frequency, sliderMax=10}
  ) {constB=2}
  n61 = SmoothStep(Value=OneMinus(None=@n57)) {constMin=0.1}
  n62 = ScalarParameter() {defaultValue=1, group=WPO Wiggle, parameterName=Wiggle Falloff, sliderMax=5}
  n63 = Transform(None=Constant3Vector() {constant=<rgba -1 0 0 0>}) {transformSourceType=TRANSFORMSOURCE_Camera}
  n64 = Constant3Vector() {constant=<rgba 0 1 0 0>}
  n65 = Normalize(VectorInput=Multiply(
    A=Transform(None=@n64) {transformSourceType=TRANSFORMSOURCE_Local, transformType=TRANSFORM_Camera}
    B=Constant3Vector() {constant=<rgba 1 1 0 0>}
   ))
  n66 = DotProduct(A=@n65, B=Constant3Vector() {constant=<rgba 0 1 0 0>})
  n67 = Transform(None=Constant3Vector() {constant=<rgba 0 1 0 0>}) {transformSourceType=TRANSFORMSOURCE_Camera}
  n68 = DotProduct(A=@n65, B=Constant3Vector() {constant=<rgba 1 0 0 0>})
  n69 = Multiply(A=Multiply(A=Sine(None=@n60), B=Power(Base=@n61, Exp=@n62)), B=Add(A=Multiply(A=@n63, B=@n66), B=Multiply(A=@n67, B=@n68)))
  n70 = ScalarParameter() {group=WPO Wiggle, parameterName=WPO Wiggle Wave}
  NamedRerouteDeclaration(None=Add(A=Add(A=@n53, B=@n56), B=Multiply(A=@n69, B=@n70))) {name=WPO, nodeColor=<rgba 0 0.3437 1 1>}
  NamedRerouteDeclaration(None=Add(A=ScalarParameter() {defaultValue=0.1, parameterName=Masked Opacity Biad}, B=@n29)) {name=Masked Opacity, nodeColor=<rgba 0.8516 1 0 1>}
  NamedRerouteUsage()
  NamedRerouteUsage()
 ]
}
